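# Additive, unlit, two-sided mesh particle material for the Geode heat shield effect.
# Two panning turbulence layers, sampled at offset UVs for color fringing, combine with crack colors into HDR emissive.
# Fresnel, V gradients, particle alpha, dynamic-parameter contrast and MF_Geode_Fade shape depth-faded opacity, while matching normal-map turbulence drives refraction.
# Time-driven sine waves along normals plus a circular wobble form a vertex offset scaled by a dynamic parameter.
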Material /Figment_S04_Geode/Athena/Effects/Materials/Geode/M_Geode_HeatShield
graph {
 properties {
  bUsedWithMeshParticles=true
  blendMode=BLEND_Additive
  pixelDepthOffsetMode=PDOM_Legacy
  refractionMethod=RM_IndexOfRefraction
  shadingModel=MSM_Unlit
  twoSided=true
  usedShadingModels=Unlit
 }
 nodes [
  n1 = MaterialFunctionCall!(In) {materialFunction=/Game/Effects/Fort_Effects/Materials/Geode/_Functions/MF_Geode_Fade}
  n2 = ParticleColor()
  n3 = Fresnel() {baseReflectFraction=0.01, exponent=0.01}
  n4 = OneMinus(None=Clamp(None=Power(Base=@n3) {constExponent=16}))
  n5 = TextureCoordinate()
  n6 = DynamicParameter() {defaultValue=<rgba 1 10 1 1>, paramNames [Scale Offset, Front Emissive Contrast, Front Opacity Contrast, Param4]}
  n7 = Power(Base=OneMinus(None=ComponentMask(None=@n5) {g=true}), Exp=LinearInterpolate(Alpha=@n6.Front Opacity Contrast) {constA=50})
  n8 = Multiply(A=@n4, B=@n7)
  n9 = TextureCoordinate() {uTiling=3, vTiling=0.25}
  n10 = AppendVector(A=ScalarParameter() {defaultValue=0.025, parameterName=U_Offset_1}, B=ScalarParameter() {defaultValue=0.025, parameterName=V_Offset_1})
  n11 = Add(A=@n9, B=@n10)
  n12 = AppendVector(
   A=TextureSample(UVs=Panner(Coordinate=@n11) {speedX=-0.5, speedY=-1}) {samplerType=SAMPLERTYPE_Normal, texture=/Game/Effects/Fort_Effects/Textures/Plasma/T_Turbulence_Seamless_N}.G
   B=TextureSample(UVs=Panner(Coordinate=@n9) {speedX=-0.5, speedY=-1}) {samplerType=SAMPLERTYPE_Normal, texture=&T_Turbulence_Seamless_N}.G
  )
  n13 = Multiply(A=@n10) {constB=-1}
  n14 = TextureSample(UVs=Panner(Coordinate=Add(A=@n9, B=@n13)) {speedX=-0.5, speedY=-1}) {samplerType=SAMPLERTYPE_Normal, texture=&T_Turbulence_Seamless_N}
  n15 = TextureCoordinate() {uTiling=2, vTiling=0.5}
  n16 = AppendVector(A=ScalarParameter() {defaultValue=0.025, parameterName=U_Offset_2}, B=ScalarParameter() {defaultValue=0.025, parameterName=V_Offset_2})
  n17 = Add(A=@n15, B=@n16)
  n18 = AppendVector(
   A=TextureSample(UVs=Panner(Coordinate=@n17) {speedX=1, speedY=-2}) {samplerType=SAMPLERTYPE_Normal, texture=&T_Turbulence_Seamless_N}.G
   B=TextureSample(UVs=Panner(Coordinate=@n15) {speedX=1, speedY=-2}) {samplerType=SAMPLERTYPE_Normal, texture=&T_Turbulence_Seamless_N}.G
  )
  n19 = Multiply(A=@n16) {constB=-1}
  n20 = TextureSample(UVs=Panner(Coordinate=Add(A=@n15, B=@n19)) {speedX=1, speedY=-2}) {samplerType=SAMPLERTYPE_Normal, texture=&T_Turbulence_Seamless_N}
  n21 = Transform(None=LinearInterpolate(A=AppendVector(A=@n12, B=@n14.G), B=AppendVector(A=@n18, B=@n20.G)))
  n22 = TextureCoordinate()
  n23 = Multiply(A=Multiply(A=Multiply(A=@n2.A, B=@n8), B=Fresnel(Normal=@n21)), B=OneMinus(None=ComponentMask(None=@n22) {g=true}))
  LinearInterpolate(B=ScalarParameter() {defaultValue=1.05, parameterName=Refraction}, Alpha=Clamp(None=Multiply(A=@n1.Opacity_HeatShield Loca, B=@n23))) {constA=1, constB=1.25}
  n24 = OneMinus(None=ComponentMask(None=TextureCoordinate()) {g=true})
  n25 = DynamicParameter() {defaultValue=<rgba 1 10 1 1>, paramNames [Scale Offset, Front Emissive Contrast, Front Opacity Contrast, Param4]}
  n26 = TextureCoordinate() {uTiling=3, vTiling=0.25}
  n27 = AppendVector(A=ScalarParameter() {defaultValue=0.025, parameterName=U_Offset}, B=ScalarParameter() {defaultValue=0.025, parameterName=V_Offset})
  n28 = Add(A=@n26, B=@n27)
  n29 = AppendVector(
   A=TextureSample(UVs=Panner(Coordinate=@n28) {speedX=-0.5, speedY=-1}) {texture=/Game/Global/Textures/Default/Caustics/T_Turbulence_Seamless}.G
   B=TextureSample(UVs=Panner(Coordinate=@n26) {speedX=-0.5, speedY=-1}) {texture=&T_Turbulence_Seamless}.G
  )
  n30 = Multiply(A=@n27) {constB=-1}
  n31 = TextureSample(UVs=Panner(Coordinate=Add(A=@n26, B=@n30)) {speedX=-0.5, speedY=-1}) {texture=&T_Turbulence_Seamless}
  n32 = TextureCoordinate() {uTiling=2, vTiling=0.5}
  n33 = AppendVector(A=ScalarParameter() {defaultValue=0.025, parameterName=U_Offset}, B=ScalarParameter() {defaultValue=0.025, parameterName=V_Offset})
  n34 = Add(A=@n32, B=@n33)
  n35 = AppendVector(
   A=TextureSample(UVs=Panner(Coordinate=@n34) {speedX=1, speedY=-2}) {texture=&T_Turbulence_Seamless}.G
   B=TextureSample(UVs=Panner(Coordinate=@n32) {speedX=1, speedY=-2}) {texture=&T_Turbulence_Seamless}.G
  )
  n36 = Multiply(A=@n33) {constB=-1}
  n37 = TextureSample(UVs=Panner(Coordinate=Add(A=@n32, B=@n36)) {speedX=1, speedY=-2}) {texture=&T_Turbulence_Seamless}
  n38 = Add(
   A=Clamp(None=Power(Base=@n24, Exp=@n25.Front Emissive Contrast))
   B=LinearInterpolate(A=AppendVector(A=@n29, B=@n31.G), B=AppendVector(A=@n35, B=@n37.G))
  )
  n39 = MaterialFunctionCall!(Use Alternate UVs, Alternate UVs) {materialFunction=/Game/Athena/Effects/Materials/Geode/Functions/MF_Geode_Crack_Colors}
  n40 = Desaturation(None=@n38)
  n41 = Power(Base=@n40) {constExponent=1.5}
  n42 = Add(A=@n38, B=LinearInterpolate(A=@n39.Cloud Color, B=@n39.Cloud Color, Alpha=Clamp(None=@n41)))
  MaterialFunctionCall(LDR Input=Clamp(None=@n42), HDR Tint=@n42, HDR Intensity Multiplier=ScalarParameter() {defaultValue=0.219, parameterName=HDR_Intensity}) {materialFunction=&DeriveHDRfromLDR}
  n43 = DynamicParameter() {defaultValue=<rgba 1 10 1 1>, paramNames [Scale Offset, Front Emissive Contrast, Front Opacity Contrast, Param4]}
  n44 = Constant()
  n45 = Multiply(A=Frac(None=Time())) {constB=50}
  n46 = OneMinus(None=ComponentMask(None=TextureCoordinate()) {g=true})
  n47 = Multiply(A=Multiply(A=@n46, B=@n46)) {constB=-2}
  n48 = Multiply(A=Sine(None=Add(A=@n45, B=@n47))) {constB=1.2, desc=Amplitude}
  n49 = Add(A=Time(), B=ComponentMask(None=TextureCoordinate()) {g=true})
  n50 = TextureCoordinate()
  n51 = Add(A=Multiply(A=Time()) {constB=150}, B=Multiply(A=ComponentMask(None=@n50) {g=true}) {constB=2})
  n52 = Add(
   A=Multiply(A=Sine(None=@n49) {period=1.31}) {constB=4, desc=Amplitude}
   B=Multiply(A=Sine(None=@n51) {period=1.31}) {constB=4, desc=Amplitude}
  )
  n53 = Multiply(A=VertexNormalWS(), B=Max(A=Add(A=@n48, B=@n52)) {constB=-2.5})
  n54 = Time()
  n55 = AppendVector(A=ComponentMask(None=WorldPosition().XYZ) {b=true}, B=Constant2Vector())
  n56 = Add(A=Multiply(A=Frac(None=@n54)) {constB=32}, B=Frac(None=Multiply(A=@n55) {constB=-0.02}))
  n57 = Constant() {r=1}
  n58 = Multiply(A=Sine(None=@n56), B=@n57) {desc=Amplitude}
  n59 = Multiply(A=Cosine(None=@n56), B=@n57)
  n60 = OneMinus(None=ComponentMask(None=TextureCoordinate()) {g=true})
  n61 = Multiply(
   A=MaterialFunctionCall(X=ComponentMask(None=@n58) {r=true}, Y=ComponentMask(None=@n59) {r=true}, Z=Constant()) {materialFunction=&MakeFloat3}.Result
   B=Multiply(A=@n60, B=@n60)
  )
  n62 = ComponentMask(None=TextureCoordinate()) {g=true}
  Multiply(A=Multiply(A=LinearInterpolate(A=@n43.Scale Offset, B=@n44), B=Add(A=@n53, B=@n61) {constB=0}), B=Clamp(None=Add(A=@n62) {constB=0.5}))
  n63 = Constant()
  n64 = Constant() {r=0.5}
  n65 = OneMinus(None=ComponentMask(None=TextureCoordinate()) {g=true})
  n66 = Clamp(None=Multiply(A=@n65, B=@n65))
  n67 = ComponentMask(None=TextureCoordinate()) {g=true}
  n68 = OneMinus(None=Multiply(A=@n67, B=@n67))
  n69 = Clamp(None=Add(A=@n68, B=@n68))
  n70 = Multiply(A=Add(A=LinearInterpolate(A=@n63, B=@n64, Alpha=@n66), B=@n40), B=Multiply(A=@n68, B=LinearInterpolate(Alpha=@n69)))
  n71 = Multiply(A=@n2.A, B=Multiply(A=Multiply(A=@n4, B=@n70), B=@n7))
  n72 = DepthFade(Opacity=@n71, FadeDistance=ScalarParameter() {defaultValue=150, parameterName=FadeDistance})
  Multiply(A=@n1.Opacity_HeatShield Loca, B=FeatureLevelSwitch(Default=@n72, ES2=@n71, ES3_1=@n71, SM4_REMOVED=@n72))
 ]
}
