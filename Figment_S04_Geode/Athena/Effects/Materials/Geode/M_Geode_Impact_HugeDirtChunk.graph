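# M_Geode_Impact_HugeDirtChunk is a default-lit material flagged for mesh particles, apparently for huge dirt chunks in the Geode impact effect.
# It holds three texture samples from the shared Athena broken-cliff-edge environment set: diffuse, mask and normal.
# The graph has no parameters or math nodes and shows no links from the samples to material outputs.
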Material /Figment_S04_Geode/Athena/Effects/Materials/Geode/M_Geode_Impact_HugeDirtChunk
graph {
 properties {bUsedWithMeshParticles=true, pixelDepthOffsetMode=PDOM_Legacy, usedShadingModels=Default Lit}
 nodes [
  TextureSample() {texture=/Game/Athena/Environments/Textures/T_CliffBroken_Edge_D}
  TextureSample() {samplerType=SAMPLERTYPE_Masks, texture=/Game/Athena/Environments/Textures/T_CliffBroken_Edge_M}
  TextureSample() {samplerType=SAMPLERTYPE_Normal, texture=/Game/Athena/Environments/Textures/T_CliffBroken_Edge_N}
 ]
}
