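# Lit material for the Geode klaxon light, drawing a tinted glass cover over a base, with a red glowing light by default.
# A UV-gradient smoothstep forms a BottomMask that blends glass and base surface values, with Fresnel darkening the glass tint.
# Emissive fakes an inner bulb by ray-tracing a sphere falloff along the local-space camera ray.
# That glow is masked off the base, scaled by Brightness and LightColor, and passed through MF_FX_EmissiveAdjustments.
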Material /Figment_S04_Geode/Athena/Effects/Materials/Geode/M_Geode_Klaxon_Light
graph {
 properties {
  pixelDepthOffsetMode=PDOM_Legacy
  previewMesh=/Game/Missions/Primary/StormShield/Meshes/S_Mission_Missle_EmergencyLight
  usedShadingModels=Default Lit
 }
 nodes [
  Fresnel() {baseReflectFraction=0, exponent=4}
  n1 = NamedRerouteUsage()
  LinearInterpolate(
   A=Multiply(A=VectorParameter() {parameterName=GlassColor, defaultValue=<rgba 0.0875 0 0 1>}.RGB, B=Fresnel() {exponent=2})
   B=VectorParameter() {parameterName=BottomColor, defaultValue=<rgba 0.1667 0.1667 0.1667 1>}.RGB
   Alpha=@n1
  )
  n2 = VectorParameter() {parameterName=GlassColor, defaultValue=<rgba 0.0875 0 0 1>}
  LinearInterpolate(A=Multiply(A=@n2.RGB, B=@n2.RGB), Alpha=NamedRerouteUsage()) {constB=0}
  LinearInterpolate(Alpha=@n1) {constA=0.2, constB=0.4}
  LinearInterpolate(Alpha=@n1) {constB=0.5}
  n3 = TransformPosition(None=CameraPositionWS()) {transformSourceType=TRANSFORMPOSSOURCE_World}
  n4 = TransformPosition(None=WorldPosition().XYZ) {transformSourceType=TRANSFORMPOSSOURCE_World}
  n5 = MaterialFunctionCall!(
   Sphere Radius=Constant() {r=1}
   Sphere Falloff=Constant() {r=32}
   Unit Ray Direction=Normalize(VectorInput=Subtract(A=@n3, B=@n4))
   Sphere Center=Constant3Vector() {constant=<rgba 0 0 12 0>}
   Ray Origin=@n3
  ) {materialFunction=/Engine/Functions/Engine_MaterialFunctions02/Gradients/RayTraceSphereFalloff}
  n6 = Power(Base=OneMinus(None=Saturate(None=@n5.Clamped Falloff))) {constExponent=4}
  n7 = Multiply(A=OneMinus(None=NamedRerouteUsage()), B=ScalarParameter() {defaultValue=1, parameterName=Brightness})
  MaterialFunctionCall(Emissive=Multiply(A=Multiply(A=@n6, B=@n7), B=VectorParameter() {parameterName=LightColor, defaultValue=<rgba 1 0.1 0.1 1>}.RGB)) {materialFunction=&MF_FX_EmissiveAdjustments}
  NamedRerouteDeclaration(None=SmoothStep(Value=MaterialFunctionCall() {materialFunction=&LinearGradient}.VGradient) {constMax=0.79, constMin=0.77}) {name=BottomMask, nodeColor=<rgba 0.4297 0 1 1>}
 ]
}
