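# Light function material for a Geode klaxon light, shaping the light into a moving warning-beacon pattern.
# A horizontal panner scrolls UVs repeated twice across U, combining a soft gradient with soft-edged elliptical hot spots via Max.
# Smoothstep masks on V fade the pattern out at one end and blend it to a constant dim level at the other.
# A separate static chain builds a symmetric vertical falloff, with no link to the animated chain.
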Material /Figment_S04_Geode/Athena/Effects/Materials/Geode/M_Geode_Klaxon_LightFunction
graph {
 properties {bAutomaticallySetUsageInEditor=false, materialDomain=MD_LightFunction, usedShadingModels=Default Lit}
 nodes [
  n1 = TextureCoordinate()
  n2 = Abs(None=ConstantBiasScale(None=ComponentMask(None=@n1) {g=true}) {bias=-0.5, scale=2})
  LinearInterpolate(
   A=LinearInterpolate(Alpha=SmoothStep(Value=@n2) {constMax=0.8}) {constA=0.1, constB=0.4}
   Alpha=SmoothStep(Value=@n2) {constMax=0.9, constMin=0.8}
  ) {constA=0.1}
  n3 = TextureCoordinate()
  n4 = Panner(Coordinate=@n3) {bFractionalPart=true, speedX=1}
  n5 = Constant2Vector() {g=1, r=2}
  n6 = Subtract(A=Frac(None=Multiply(A=@n4, B=@n5)), B=Constant2Vector() {g=0.5, r=0.5})
  n7 = Abs(None=@n6)
  n8 = LinearInterpolate(Alpha=OneMinus(None=ComponentMask(None=@n7) {r=true})) {constA=0.1, constB=0.4}
  n9 = Multiply(A=@n6, B=Constant2Vector() {g=3, r=3.1416})
  n10 = SmoothStep(Value=DotProduct(A=@n9, B=@n9)) {constMax=0.05, constMin=0.25}
  n11 = ComponentMask(None=@n3) {g=true}
  LinearInterpolate(
   A=Multiply(A=Max(A=@n8, B=@n10), B=SmoothStep(Value=@n11) {constMax=0.7, constMin=0.9})
   Alpha=SmoothStep(Value=@n11) {constMax=0.2, constMin=0.4}
  ) {constB=0.4}
 ]
}
