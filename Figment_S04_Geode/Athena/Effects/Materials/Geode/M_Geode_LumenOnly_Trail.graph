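# Trail effect material for the Geode effects, named as Lumen-only, that computes a bright HDR glow, orange by default.
# The V texture coordinate runs through an inverted SmoothStep, peaking at V=0 and fading out within the first quarter.
# A high power tightens that falloff into a narrow band, which is multiplied by the EmissiveColor parameter.
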Material /Figment_S04_Geode/Athena/Effects/Materials/Geode/M_Geode_LumenOnly_Trail
graph {
 properties {bAutomaticallySetUsageInEditor=false, usedShadingModels=Default Lit}
 nodes [
  n1 = ComponentMask(None=TextureCoordinate()) {g=true}
  Multiply(
   A=Power(Base=SmoothStep(Value=@n1) {constMax=0, constMin=0.25}) {constExponent=16}
   B=VectorParameter() {parameterName=EmissiveColor, defaultValue=<rgba 100 29.6335 5.8171 1>}.RGB
  )
 ]
}
