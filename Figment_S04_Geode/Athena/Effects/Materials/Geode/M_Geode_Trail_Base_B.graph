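# Masked, two-sided base material for Geode trail effects on mesh particles and spline meshes, giving a billowing cloud look with HDR glow.
# Panning storm-cloud masks and normals on world-Z-tiled coordinates feed named reroutes for surface shading and deformation.
# Vertex offset expands along normals, scaled by height below a collection ZHeight, with a collapse blend and sine-based mobile displacement.
# Opacity combines view-angle edge erosion, a cutoff and non-mobile near-camera fade.
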Material /Figment_S04_Geode/Athena/Effects/Materials/Geode/M_Geode_Trail_Base_B
graph {
 properties {
  bUsedWithMeshParticles=true
  bUsedWithSplineMeshes=true
  blendMode=BLEND_Masked
  pixelDepthOffsetMode=PDOM_Legacy
  translucencyLightingMode=TLM_SurfacePerPixelLighting
  twoSided=true
  usedShadingModels=Default Lit
 }
 nodes [
  Constant()
  Constant() {r=1}
  n1 = NamedRerouteUsage()
  n2 = ScalarParameter() {defaultValue=0.25, parameterName=ColorIntensity}
  Multiply(
   A=Saturate(None=Power(Base=@n1, Exp=@n2))
   B=VectorParameter() {group=BaseColor, parameterName=BaseColor, defaultValue=<rgba 0.7 0.7 0.7 1>}.RGB
  )
  NamedRerouteDeclaration(None=Multiply(A=Time(), B=ScalarParameter() {defaultValue=0.07, group=TextureControl, parameterName=Speed})) {name=Time Scaled, nodeColor=<rgba 1 0.7969 0 1>}
  n3 = NamedRerouteUsage()
  n4 = Multiply(A=NamedRerouteUsage(), B=ScalarParameter() {defaultValue=0.2, parameterName=LargestScaleMultiply})
  n5 = NamedRerouteUsage()
  n6 = TextureSample(UVs=Panner(Coordinate=AppendVector(A=@n3, B=@n4), Time=@n5) {bFractionalPart=true, speedX=0.05, speedY=-0.1}) {samplerType=SAMPLERTYPE_LinearGrayscale, texture=/Game/VisualThreatManager/StormVisuals/LargeClouds/T_StormCloudDetail01_M}
  n7 = NamedRerouteUsage()
  n8 = Add(A=AppendVector(A=NamedRerouteUsage(), B=Multiply(A=@n7) {constB=0.4})) {constB=0.4}
  n9 = ScalarParameter() {defaultValue=10, parameterName=FastPannerSpeedU}
  n10 = ScalarParameter() {defaultValue=-20, parameterName=FastPannerSpeedV}
  n11 = TextureSample(UVs=Panner(Coordinate=@n8, Time=@n5, Speed=AppendVector(A=@n9, B=@n10)) {bFractionalPart=true, speedX=20, speedY=-80}) {samplerType=SAMPLERTYPE_LinearGrayscale, texture=&T_StormCloudDetail01_M}
  NamedRerouteDeclaration(None=Add(A=@n6.G, B=@n11.G)) {name=Vertical Pan B, nodeColor=<rgba 0.4115 0.4115 0.4115 1>}
  n12 = NamedRerouteUsage()
  n13 = NamedRerouteUsage()
  n14 = Panner(Coordinate=AppendVector(A=NamedRerouteUsage(), B=Multiply(A=@n12) {constB=0.6}), Time=@n13) {bFractionalPart=true, speedX=-0.06, speedY=-1}
  n15 = TextureSample(UVs=@n14) {samplerType=SAMPLERTYPE_LinearGrayscale, texture=&T_StormCloudDetail01_M}
  NamedRerouteDeclaration(None=LinearInterpolate(A=Multiply(A=@n6.G, B=@n15.G), B=Multiply(A=@n11.G, B=@n15.G), Alpha=NamedRerouteUsage())) {name=Vertical Pan C, desc=for cloud noise, nodeColor=<rgba 0.1198 0.1198 0.1198 1>}
  n16 = Power(
   Base=MaterialFunctionCall() {materialFunction=&LinearGradient}.VGradient
   Exp=ScalarParameter() {defaultValue=1, group=WPO, parameterName=ExpandPower}
  )
  n17 = VertexNormalWS()
  n18 = ScalarParameter() {defaultValue=500, parameterName=MacroWidthAmount}
  n19 = Subtract(
   A=CollectionParameter() {collection=/Game/Effects/Fort_Effects/Materials/Geode/MPC_GeodeNick, parameterName=ZHeight}
   B=MaterialFunctionCall(Float3=WorldPosition().XYZ) {materialFunction=&BreakOutFloat3Components}.B
  )
  n20 = Divide(A=@n19, B=ScalarParameter() {defaultValue=342983.5, parameterName=MacroScale})
  n21 = Saturate(None=@n20)
  n22 = Add(
   A=Multiply(A=Multiply(A=@n16, B=@n17), B=ScalarParameter() {group=WPO, parameterName=Expand})
   B=Multiply(A=VertexNormalWS(), B=Multiply(A=@n18, B=@n21))
  )
  n23 = ScalarParameter() {defaultValue=1, group=WPO, parameterName=Valley}
  n24 = Saturate(None=Multiply(A=@n20, B=ScalarParameter() {defaultValue=1.7783, parameterName=FalloffHardness}))
  n25 = LinearInterpolate(A=ScalarParameter() {defaultValue=2.5, parameterName=PeaksMin}, B=ScalarParameter() {defaultValue=4, parameterName=PeaksMax}, Alpha=@n24)
  n26 = NamedRerouteUsage()
  n27 = Multiply(
   A=Multiply(A=LinearInterpolate(A=@n23, B=@n25, Alpha=@n26) {constA=-1, constB=2}, B=VertexNormalWS())
   B=ScalarParameter() {defaultValue=200, group=WPO, parameterName=Deform, sortPriority=2}
  )
  n28 = ScalarParameter() {defaultValue=1.5, group=WPO, parameterName=ValleyMobile}
  n29 = LinearInterpolate(
   A=ScalarParameter() {defaultValue=2, parameterName=PeaksMinMobile}
   B=ScalarParameter() {defaultValue=2.2, parameterName=PeaksMaxMobile}
   Alpha=@n24
  )
  n30 = NamedRerouteUsage()
  n31 = Divide(A=@n30, B=ScalarParameter() {defaultValue=0.0644, parameterName=SineTiling})
  n32 = Divide(A=@n30, B=ScalarParameter() {defaultValue=0.1215, parameterName=SineTiling2})
  n33 = Add(
   A=ConstantBiasScale(None=Sine(None=@n31))
   B=Multiply(A=Sine(None=@n32), B=ScalarParameter() {defaultValue=0.33, parameterName=SecondSineMulti})
  )
  n34 = Multiply(
   A=Multiply(A=LinearInterpolate(A=@n28, B=@n29, Alpha=@n33) {constA=-1, constB=2}, B=VertexNormalWS())
   B=ScalarParameter() {defaultValue=200, group=WPO, parameterName=DeformMobile, sortPriority=2}
  )
  n35 = Multiply(
   A=Add(A=@n24, B=ScalarParameter() {defaultValue=1, parameterName=MinMacroMultiply})
   B=FeatureLevelSwitch(Default=@n27, ES2=@n34, ES3_1=@n34, SM4_REMOVED=@n27)
  ) {constA=1}
  n36 = VertexNormalWS()
  n37 = ScalarParameter() {defaultValue=-0.3937, parameterName=NormalCollapseAmount}
  NamedRerouteDeclaration(None=LinearInterpolate(A=Add(A=@n22, B=@n35), B=Multiply(A=@n36, B=@n37), Alpha=NamedRerouteUsage())) {name=WPO Output, nodeColor=<rgba 0 0.237 1 1>}
  n38 = NamedRerouteUsage()
  n39 = NamedRerouteUsage()
  n40 = NamedRerouteUsage()
  n41 = Multiply(
   A=Multiply(A=@n38, B=LinearInterpolate(A=@n39, B=@n40, Alpha=@n38))
   B=VectorParameter() {group=Glow, parameterName=GlowColor_1, defaultValue=<rgba 5 1 0 1>}.RGB
  )
  n42 = Multiply(A=ScalarParameter() {group=Glow, parameterName=SubsurfaceBoost}, B=@n41)
  n43 = ScalarParameter() {defaultValue=0.1, group=Glow, parameterName="Desaturate Luminance (S)_1"}
  n44 = ScalarParameter() {defaultValue=0.6, group=Glow, parameterName=SubSurfaceAmount_1}
  n45 = OneMinus(None=Multiply(A=NamedRerouteUsage()) {constB=1.5})
  n46 = Multiply(A=@n41, B=Clamp(None=LinearInterpolate(A=@n44, Alpha=@n45) {constA=0.25}))
  n47 = Add(
   A=MaterialFunctionCall(LDR Input=Clamp(None=@n42), HDR Tint=@n42, Desaturate Luminance=@n43) {materialFunction=&DeriveHDRfromLDR}.RemappedRange
   B=MaterialFunctionCall(LDR Input=Saturate(None=@n46), HDR Tint=@n46, Desaturate Luminance=@n43) {materialFunction=&DeriveHDRfromLDR}.RemappedRange
  )
  NamedRerouteDeclaration(None=FeatureLevelSwitch(Default=@n47, ES2=@n46, ES3_1=@n46, SM4_REMOVED=@n47)) {name=Emissive Output, nodeColor=<rgba 0 1 0.5312 1>}
  n48 = TextureCoordinate() {uTiling=0.5}
  NamedRerouteDeclaration(None=Multiply(
    A=MaterialFunctionCall(Float2=@n48) {materialFunction=&BreakOutFloat2Components}.R
    B=ScalarParameter() {defaultValue=2, group=TextureControl, parameterName=ScaleU}
   )) {name=Tex U, nodeColor=<rgba 0.5341 1 0.6168 1>}
  n49 = MaterialFunctionCall(Float3=WorldPosition().XYZ) {materialFunction=&BreakOutFloat3Components}
  n50 = ScalarParameter() {defaultValue=-62648.6055, parameterName=ZWorldTiling}
  n51 = ScalarParameter() {group=TextureControl, parameterName=ScaleV}
  n52 = ScalarParameter() {defaultValue=2, parameterName=TileMultiply}
  NamedRerouteDeclaration(None=Multiply(A=Divide(A=@n49.B, B=@n50), B=Multiply(A=@n51, B=@n52))) {name=World Tile VScaled, nodeColor=<rgba 0.0111 0.6967 1 1>}
  n53 = ScalarParameter() {defaultValue=3006.0029, parameterName=VertCollapseFalloff}
  n54 = Divide(
   A=NamedRerouteDeclaration(None=Divide(A=@n19, B=@n53)) {name=Vertical Falloff, nodeColor=<rgba 1 0 0.105 1>}
   B=ScalarParameter() {defaultValue=4.8567, parameterName=HotLipFalloff}
  )
  NamedRerouteDeclaration(None=OneMinus(None=Clamp(None=@n54))) {name=Rev Vertical Falloff, nodeColor=<rgba 1 0.3095 0 1>}
  n55 = Panner(Coordinate=@n8, Time=@n13) {bFractionalPart=true, speedX=0.05, speedY=-1}
  n56 = Multiply(
   A=VectorParameter() {parameterName=Normallntensity, defaultValue=<rgba 0.2 0.2 1 1>}.RGB
   B=MaterialFunctionCall(
    BaseNormal=TextureSample(UVs=@n55) {samplerType=SAMPLERTYPE_Normal, texture=/Game/VisualThreatManager/StormVisuals/LargeClouds/T_StormCloudDetail01_N}.RGB
    AdditionalNormal=TextureSample(UVs=@n14) {samplerType=SAMPLERTYPE_Normal, texture=&T_StormCloudDetail01_N}.RGB
   ) {materialFunction=&BlendAngleCorrectedNormals}.Result
  )
  NamedRerouteDeclaration(None=FeatureLevelSwitch(Default=@n56, ES2=@n56, ES3_1=@n56, SM4_REMOVED=@n56)) {name=Normal Output, nodeColor=<rgba 0.243 0.224 1 1>}
  n57 = Divide(
   A=Add(A=PixelDepth().R, B=ScalarParameter() {defaultValue=-400, parameterName=NearClipBias})
   B=ScalarParameter() {defaultValue=100, parameterName=NearClipDistance}
  )
  n58 = Constant() {r=1}
  n59 = MaterialFunctionCall!(
   Diffuse Texture Object=TextureObject() {samplerType=SAMPLERTYPE_LinearColor, texture=/Game/Global/Textures/Default/Tileable_Patterns/Cloud_Diffuse}
   Spec Texture Object
   TextureSize=Multiply(A=VectorParameter() {parameterName=TextureSIze, defaultValue=<rgba 5000 5000 6000 1>}.RGB) {constB=3}
   --------------------------
   Use seperate settings for Z projection
   Use seperate texture for Z projection
   Texture Size Z projection
   Z Plane Diffuse Texture Object
   Z Plane Spec Texture Object
   "--------------------------      "
   ProjectionTransitionContrast
   Use Texture Alpha Heightmap to Refine the Transistion
   HeightLerpTransitionContrast
   PickTheDiffuseHeightChannel
   TheDiffuseHeightChannel
   "--------------------------       "
   NormalTextureObject=TextureObject() {samplerType=SAMPLERTYPE_Normal, texture=/Game/Global/Textures/Default/Tileable_Patterns/Cloud_normals}
   World Space Normals
   "--------------------------  "
   Export Float 4 Diffuse
   Export Float 4 Spec
   "--------------------------     "
   WorldPosition=WorldPosition().XYZ
   World Space Normal
  ) {materialFunction=/Engine/Functions/Engine_MaterialFunctions02/Texturing/WorldAlignedTextures_Complex}
  n60 = VertexNormalWS()
  n61 = Constant() {desc=.2, r=0.2}
  n62 = Clamp(None=DotProduct(A=LinearInterpolate(A=@n59.Normal XYZ Texture, B=@n60, Alpha=@n61) {desc=Soft normal}, B=CameraVectorWS()))
  n63 = Clamp(None=DotProduct(A=VertexNormalWS(), B=CameraVectorWS()))
  n64 = MaterialFunctionCall(Transition Phase=ScalarParameter() {defaultValue=1, parameterName=Opacity}, Height Texture=Divide(A=Add(A=@n62, B=@n63))) {materialFunction=&HeightLerp}
  n65 = Constant() {r=1}
  n66 = Constant()
  n67 = If(A=NamedRerouteUsage(), B=@n65, "A > B"=@n66, "A == B"=@n66, "A < B"=@n65)
  NamedRerouteDeclaration(None=Multiply(A=FeatureLevelSwitch(Default=@n57, ES2=@n58, ES3_1=@n58, SM4_REMOVED=@n57), B=Multiply(A=@n64.Alpha, B=@n67))) {name=Opacity Output, nodeColor=<rgba 0.875 0 1 1>}
  n68 = Subtract(
   A=Add(
    A=MaterialFunctionCall() {materialFunction=&ObjectPivotPoint}.Object Pivot Location
    B=VectorParameter() {parameterName=Pinch Offset, defaultValue=<rgba 0 0 5750 0>}.RGB
   )
   B=WorldPosition().XYZ
  )
  n69 = Saturate(None=Divide(A=ComponentMask(None=@n68) {b=true}, B=ScalarParameter() {defaultValue=1250, parameterName=Falloff Bottom}))
  n70 = NamedRerouteUsage()
  n71 = Power(Base=OneMinus(None=Saturate(None=@n70)), Exp=ScalarParameter() {defaultValue=2, parameterName=pow})
  n72 = LinearInterpolate(A=@n71, B=Add(A=@n71, B=Multiply(A=@n69, B=@n69)), Alpha=ScalarParameter() {parameterName=Bottom Close})
  NamedRerouteDeclaration(None=FeatureLevelSwitch(Default=@n72, ES2=@n71, ES3_1=@n71, SM4_REMOVED=@n72)) {name=Bottom Close Alpha, nodeColor=<rgba 0 0.0299 1 1>}
  NamedRerouteDeclaration(None=Add(A=@n6.G, B=@n15.G)) {name=Vertical Pan A, nodeColor=<rgba 1 1 1 1>}
  NamedRerouteUsage()
  NamedRerouteUsage()
  NamedRerouteUsage()
  NamedRerouteUsage()
 ]
}
