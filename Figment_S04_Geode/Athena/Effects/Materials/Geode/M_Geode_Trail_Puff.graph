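# Masked, lit mesh-particle material for a Geode trail puff, shaping meshes into billowing, glowing smoke.
# Panned storm-cloud samples drive base color, normals and, with 3D noise and gradient-weighted expansion, vertex displacement that collapses to particle center near camera.
# View-facing terms from world-aligned cloud normals feed a particle-alpha HeightLerp for erosion.
# Particle red and blue set glow radius and intensity; dynamic parameters control panning, tiling, deformation and emissive blend.
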Material /Figment_S04_Geode/Athena/Effects/Materials/Geode/M_Geode_Trail_Puff
graph {
 properties {
  bUsedWithMeshParticles=true
  blendMode=BLEND_Masked
  pixelDepthOffsetMode=PDOM_Legacy
  translucencyLightingMode=TLM_VolumetricDirectional
  usedShadingModels=Default Lit
 }
 nodes [
  Constant()
  Constant() {r=1}
  Constant()
  n1 = ScalarParameter() {defaultValue=-1, group=WPO, parameterName=Valley}
  n2 = ScalarParameter() {defaultValue=2, group=WPO, parameterName=Peaks}
  n3 = ScalarParameter() {defaultValue=1, group=TextureControl, parameterName=ScaleU}
  n4 = MaterialFunctionCall(Float2=TextureCoordinate()) {materialFunction=&BreakOutFloat2Components}
  n5 = Multiply(A=@n3, B=@n4.R)
  n6 = Multiply(
   A=ScalarParameter() {defaultValue=1, group=TextureControl, parameterName=ScaleV}
   B=DynamicParameter() {paramNames [PannerSpeed, TileMultiply, DeformAmount, EmissiveControl]}.TileMultiply
  )
  n7 = Multiply(A=@n4.G, B=@n6)
  n8 = Time()
  n9 = Constant2Vector() {g=1, r=0.05}
  n10 = DynamicParameter() {paramNames [PannerSpeed, TileMultiply, DeformAmount, EmissiveControl]}
  n11 = TextureSample(UVs=Panner(Coordinate=AppendVector(A=@n5, B=@n7), Time=@n8, Speed=Multiply(A=@n9, B=@n10.PannerSpeed)) {bFractionalPart=true, speedX=0.05, speedY=1}) {samplerType=SAMPLERTYPE_LinearGrayscale, texture=/Game/VisualThreatManager/StormVisuals/LargeClouds/T_StormCloudDetail01_M}
  n12 = MaterialFunctionCall(Float2=TextureCoordinate() {vTiling=2}) {materialFunction=&BreakOutFloat2Components}
  n13 = Panner(
   Coordinate=AppendVector(A=Multiply(A=@n3, B=@n12.R), B=Multiply(A=@n12.G, B=@n6))
   Time=@n8
   Speed=Multiply(A=Constant2Vector() {g=1, r=0.05}, B=@n10.PannerSpeed)
  ) {bFractionalPart=true, speedX=0.05, speedY=1}
  n14 = TextureSample(UVs=@n13) {samplerType=SAMPLERTYPE_LinearGrayscale, texture=&T_StormCloudDetail01_M}
  n15 = Add(A=@n11.G, B=@n14.G)
  n16 = ScalarParameter() {defaultValue=1, group=WPO, parameterName=LargeScaleDistort}
  n17 = MaterialFunctionCall(Float2=TextureCoordinate() {uTiling=3, vTiling=5}) {materialFunction=&BreakOutFloat2Components}
  n18 = Panner(
   Coordinate=AppendVector(A=Multiply(A=@n3, B=@n17.R), B=Multiply(A=@n17.G, B=@n6))
   Time=@n8
   Speed=Multiply(A=Constant2Vector() {g=1, r=-0.1}, B=@n10.PannerSpeed)
  ) {bFractionalPart=true, speedX=-0.1, speedY=1}
  n19 = Add(
   A=@n14.G
   B=TextureSample(UVs=@n18) {samplerType=SAMPLERTYPE_LinearGrayscale, texture=&T_StormCloudDetail01_M}.G
  )
  n20 = Add(
   A=Multiply(
    A=MaterialFunctionCall() {materialFunction=&LinearGradient}.VGradient
    B=Multiply(A=@n15, B=@n16)
   )
   B=Add(A=@n11.G, B=@n19)
  )
  n21 = Multiply(
   A=Multiply(A=LinearInterpolate(A=@n1, B=@n2, Alpha=@n20) {constA=-1, constB=2}, B=VertexNormalWS())
   B=DynamicParameter() {paramNames [PannerSpeed, TileMultiply, DeformAmount, EmissiveControl]}.DeformAmount
  )
  n22 = MaterialFunctionCall() {materialFunction=&LinearGradient}
  n23 = Time()
  n24 = ScalarParameter() {defaultValue=10, group=WPO, parameterName=3DPanSpeed}
  n25 = Multiply(A=Add(A=WorldPosition().XYZ, B=Multiply(A=@n23, B=@n24)), B=ScalarParameter() {defaultValue=0.0025, group=WPO, parameterName=3DNoiseTile}) {constB=0.01}
  n26 = Multiply(
   A=@n22.VGradient
   B=Multiply(
    A=VectorNoise(World Position=@n25) {noiseFunction=VNF_VectorALU, tileSize=1}
    B=ScalarParameter() {defaultValue=50, group=WPO, parameterName=3DNoiseScale}
   )
  )
  n27 = Multiply(A=Power(Base=@n22.VGradient, Exp=ScalarParameter() {defaultValue=1, group=WPO, parameterName=ExpandPower}), B=VertexNormalWS())
  n28 = ScalarParameter() {group=WPO, parameterName=Expand}
  LinearInterpolate(
   A=Subtract(A=ParticlePositionWS(), B=WorldPosition().XYZ)
   B=Add(A=Add(A=@n21, B=@n26), B=Multiply(A=@n27, B=@n28))
   Alpha=MaterialFunctionCall(Distance=ScalarParameter() {defaultValue=350, parameterName="Distance (S)"}) {materialFunction=&Near_Camera_Fade}.Result
  )
  n29 = Time()
  n30 = Multiply(B=VectorParameter() {parameterName=Panner, defaultValue=<rgba 0 0 300 1>}.RGB)
  n31 = MaterialFunctionCall!(
   Diffuse Texture Object=TextureObject() {samplerType=SAMPLERTYPE_LinearColor, texture=/Game/Global/Textures/Default/Tileable_Patterns/Cloud_Diffuse}
   Spec Texture Object
   TextureSize=Multiply(A=VectorParameter() {parameterName=TextureSIze, defaultValue=<rgba 600 600 800 1>}.RGB) {constB=3}
   --------------------------
   Use seperate settings for Z projection
   Use seperate texture for Z projection
   Texture Size Z projection
   Z Plane Diffuse Texture Object
   Z Plane Spec Texture Object
   "--------------------------      "
   ProjectionTransitionContrast
   Use Texture Alpha Heightmap to Refine the Transistion
   HeightLerpTransitionContrast
   PickTheDiffuseHeightChannel
   TheDiffuseHeightChannel
   "--------------------------       "
   NormalTextureObject=TextureObject() {samplerType=SAMPLERTYPE_Normal, texture=/Game/Global/Textures/Default/Tileable_Patterns/Cloud_normals}
   World Space Normals
   "--------------------------  "
   Export Float 4 Diffuse
   Export Float 4 Spec
   "--------------------------     "
   WorldPosition=Add(A=Multiply(A=@n29, B=@n30) {constB=200}, B=WorldPosition().XYZ)
   World Space Normal
  ) {materialFunction=/Engine/Functions/Engine_MaterialFunctions02/Texturing/WorldAlignedTextures_Complex}
  n32 = VertexNormalWS()
  n33 = Constant() {desc=.2, r=0.2}
  n34 = Clamp(None=DotProduct(A=LinearInterpolate(A=@n31.Normal XYZ Texture, B=@n32, Alpha=@n33) {desc=Soft normal}, B=CameraVectorWS()))
  n35 = Clamp(None=DotProduct(A=VertexNormalWS(), B=CameraVectorWS()))
  MaterialFunctionCall(Transition Phase=ParticleColor().A, Height Texture=Divide(A=Add(A=@n34, B=@n35))) {materialFunction=&HeightLerp}
  n36 = ScalarParameter() {group=Glow, parameterName=DistortSphereMask}
  n37 = Multiply(
   A=SphereMask(
    A=Multiply(A=@n19, B=@n36)
    B=VectorParameter() {parameterName=CenterGlow, defaultValue=<rgba 0.5 0.5 0 1>}.RGB
    Radius=ParticleColor().R
    Hardness=ScalarParameter() {group=Glow, parameterName=Hardness}
   )
   B=Multiply(A=VectorParameter() {group=Glow, parameterName=GlowColor, defaultValue=<rgba 250 60 0 1>}.RGB, B=ParticleColor().B)
  )
  n38 = Multiply(A=ScalarParameter() {group=Glow, parameterName=SubsurfaceBoost}, B=@n37)
  n39 = ScalarParameter() {defaultValue=0.25, group=Glow, parameterName="Desaturate Luminance (S)"}
  n40 = MaterialFunctionCall(LDR Input=Clamp(None=@n38), HDR Tint=@n38, Desaturate Luminance=@n39) {materialFunction=&DeriveHDRfromLDR}
  n41 = ScalarParameter() {defaultValue=0.5, group=Glow, parameterName=SubSurfaceAmount}
  n42 = OneMinus(None=Multiply(A=@n19) {constB=1.5})
  n43 = Multiply(A=Clamp(None=LinearInterpolate(A=@n41, Alpha=@n42) {constA=0.25}), B=@n37)
  n44 = MaterialFunctionCall(LDR Input=Clamp(None=@n43), HDR Tint=@n43, Desaturate Luminance=@n39) {materialFunction=&DeriveHDRfromLDR}
  n45 = VectorParameter() {group=Glow, parameterName=GlowColor_1, defaultValue=<rgba 25 6.0293 0 1>}
  MaterialFunctionCall(Emissive=LinearInterpolate(
    A=Add(A=@n40.RemappedRange, B=@n44.RemappedRange)
    B=Multiply(A=@n45.RGB, B=@n19)
    Alpha=DynamicParameter() {paramNames [PannerSpeed, TileMultiply, DeformAmount, EmissiveControl]}.EmissiveControl
   )) {materialFunction=&MF_FX_EmissiveAdjustments}
  Multiply(
   A=VectorParameter() {parameterName=NormalMultiply, defaultValue=<rgba 0.25 0.25 1 1>}.RGB
   B=MaterialFunctionCall(
    BaseNormal=TextureSample(UVs=@n13) {samplerType=SAMPLERTYPE_Normal, texture=/Game/VisualThreatManager/StormVisuals/LargeClouds/T_StormCloudDetail01_N}.RGB
    AdditionalNormal=TextureSample(UVs=@n18) {samplerType=SAMPLERTYPE_Normal, texture=&T_StormCloudDetail01_N}.RGB
   ) {materialFunction=&BlendAngleCorrectedNormals}.Result
  )
  Multiply(
   A=Clamp(None=Power(Base=@n19) {constExponent=1})
   B=VectorParameter() {group=BaseColor, parameterName=BaseColor, defaultValue=<rgba 0.7 0.7 0.7 1>}.RGB
  )
  ScalarParameter() {defaultValue=1, parameterName=Opacity}
 ]
}
